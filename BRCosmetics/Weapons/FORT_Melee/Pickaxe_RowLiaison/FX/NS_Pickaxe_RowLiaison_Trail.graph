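# Niagara swing trail effect for the RowLiaison pickaxe, layering three anim-trail ribbons (soft, bright, edge) with two socket-sampled sprite emitters.
# Glitter spawns by distance moved and flickers through random RGB scaling, while Sparkles spawns at distance-scaled rates in two size groups, both drifting on curl noise.
# User.VariantPicker selects the color variant per emitter, with time-of-day tinting throughout and a front-end alpha modifier on the bright and edge ribbons.
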NiagaraSystem /BRCosmetics/Weapons/FORT_Melee/Pickaxe_RowLiaison/FX/NS_Pickaxe_RowLiaison_Trail
dependencies {
 usedDataInterfaces [
  &NiagaraDataInterfaceArrayColor
  &NiagaraDataInterfaceSkeletalMesh
  &NiagaraDataInterfaceCurve
  &NiagaraDataInterfaceColorCurve
  &NiagaraDataInterfaceArrayFloat
  &NiagaraDataInterfaceSocketReader
 ]
 usedDynamicInputs [&SetBoolByIntComparison]
 usedModules [
  &self:SystemSpawnScript.SystemScriptSource.SystemScriptGraph.NiagaraNodeAssignment_0.SetVariables_#1
  &self:SystemSpawnScript.SystemScriptSource.SystemScriptGraph.NiagaraNodeAssignment_4.SetVariables_#2
  &self:SystemSpawnScript.SystemScriptSource.SystemScriptGraph.NiagaraNodeAssignment_2.SetVariables_#3
  &self:SystemSpawnScript.SystemScriptSource.SystemScriptGraph.NiagaraNodeAssignment_3.SetVariables_#4
  &self:SystemSpawnScript.SystemScriptSource.SystemScriptGraph.NiagaraNodeAssignment_1.SetVariables_#5
  &SystemState
 ]
}
emitters {
 Glitter {
  emitterUpdateScript [
   &EmitterState()
   &SpawnPerUnit(Module.Use Max Movement Threshold=true, Module.Use Movement Tolerance=false, Spawn Spacing=50, Use Spawn Probability=false)
  ]
  particleSpawnScript [
   &InitializeParticle(
    Color=&MergeHelper_ColorSwitch(
     0=&RandomRangeLinearColor(Minimum=<rgba 4 2.5287 0.4 1>, Maximum=<rgba 5 0.9696 0.5 1>)
     1=&RandomRangeLinearColor(Minimum=<rgba 1.1354 1.5264 2 0.9957>, Maximum=<rgba 0.5156 1.2502 3 1>)
     2=&RandomRangeLinearColor(Minimum=<rgba 1 0.814 0.7947 1>, Maximum=<rgba 2 1.6231 1.5861 1>)
     3=&RandomRangeLinearColor(Minimum=<rgba 0.9323 0.977 1 0.9957>, Maximum=<rgba 0.2 1.0139 2 1>)
     4=&RandomRangeLinearColor(Minimum=<rgba 2 0.8646 1.6347 0.9957>, Maximum=<rgba 5 1.5104 2.967 1>)
     5=&RandomRangeLinearColor(Minimum=<rgba 1.952 1.3112 2 0.9957>, Maximum=<rgba 5 1.5104 2.967 1>)
     Index=&MakeIntFromFloat(Float=$User.VariantPicker, Conversion Mode=::NewEnumerator1)
     Num=::NewEnumerator1
    )
    Lifetime Max=0.4
    Lifetime Min=0.3
    Lifetime Mode=::NewEnumerator1
    Sprite Size Mode=::NewEnumerator1
    Uniform Sprite Size Max=1.5
    Uniform Sprite Size Min=0.8
    UsePositionOffset=false
    Write Material Random=true
   )
   &InheritVelocity(
    Inherited Velocity Speed Limit:NiagaraFloat=&RandomRangeFloat(Minimum=40, Maximum=80)
    Speed Limit=true
    UseSpeedThreshold=false
   )
   &ApplyInitialForces()
   &ShapeLocation(MaskSpawnGroup=false, Override Seed=false, Sphere Radius=10, UseOwnerScale=true)
   &SocketLocation(SamplePosOffset=false, Socket Reader=$System.Socket Reader, Socket Sampling Mode=::NewEnumerator3, VelocitySpeedLimit=false)
  ]
  particleUpdateScript [
   &ParticleState()
   &ScaleSpriteSize(Scale Sprite Size Mode=::NewEnumerator2) {module=ScaleSpriteSize001}
   &ScaleSpriteSize(Uniform Curve Sprite Scale=DataInterface() {propertyValues {Curve {keys [
        {tangentMode=RCTM_User, time=0, value=1}
        {arriveTangent=-0.0405, interpMode=RCIM_Cubic, leaveTangent=-0.0405, tangentMode=RCTM_User, time=0.2, value=1}
        {arriveTangent=-2.1547, interpMode=RCIM_Cubic, leaveTangent=-2.1547, tangentMode=RCTM_User, time=1, value=0}
       ]}}})
   &CurlNoiseForce(Noise Frequency=20, Noise Strength=200, Pan Noise=false)
   &Drag(Drag=&RandomRangeFloat(Minimum=6, Maximum=9), Use Linear Drag=true, Use Rotational Drag=true)
   &SolveForcesAndVelocity(Clamp Velocity=false, Limit Acceleration=false)
   /SetVariables(
    Particles.Color2=&Multiply_LinearColorByFloat(
     Linear Color=&MergeHelper_ColorSwitch(
      0=&RandomRangeLinearColor(Minimum=<rgba 0.0005 0.0008 0.001 1>, Maximum=<rgba 0.1 0.0383 0 1>, Evaluation Type=::NewEnumerator1)
      1=&RandomRangeLinearColor(Minimum=<rgba 0.0005 0.0008 0.001 1>, Maximum=<rgba 0.0163 0.0541 0.125 1>, Evaluation Type=::NewEnumerator1)
      2=&RandomRangeLinearColor(Minimum=<rgba 0.0005 0.0008 0.001 1>, Maximum=<rgba 0.1022 0.0719 0.08 1>, Evaluation Type=::NewEnumerator1)
      3=&RandomRangeLinearColor(Minimum=<rgba 0.0005 0.0008 0.001 1>, Maximum=<rgba 0 0.0383 0.1 1>, Evaluation Type=::NewEnumerator1)
      4=&RandomRangeLinearColor(Minimum=<rgba 0.0005 0.0008 0.001 1>, Maximum=<rgba 0.07 0.0256 0.07 1>, Evaluation Type=::NewEnumerator1)
      5=&RandomRangeLinearColor(Minimum=<rgba 0.0005 0.0008 0.001 1>, Maximum=<rgba 0.07 0.0256 0.07 1>, Evaluation Type=::NewEnumerator1)
      Index=&MakeIntFromFloat(Float=$User.VariantPicker, Conversion Mode=::NewEnumerator1)
      Num=::NewEnumerator1
     )
     Float=&FN_Select_Float_from_Time_Of_Day(Morning=1, Day=1, Evening=1, Night=0.8)
    )
    Particles.SpriteSize2=&Vector2DFromFloat(Value=10)
   ) {isSetParameters=true}
   &ScaleColor(
    Scale RGB=&VectorFromFloat(Value=&Multiply_Float(
      A=&RandomRangeFloat(Minimum=-10, Maximum=10, Evaluation Type=::NewEnumerator1)
      B=&MakeCustomFloatFromBool(Bool=$User.bIsFrontEndPreview, True Float=1, False Float=1)
     ))
    ScaleA=true
    ScaleRGB=true
   )
   &ScaleColor(
    Scale Alpha=&FNTimeOfDayScalar(Night Scale=1, Morning Scale=1, Evening Scale=1)
    Scale RGB=&MakeVectorFromLinearColorRGB(Color=$System.TOD)
    ScaleA=true
    ScaleRGB=true
   ) {module=ScaleColor001}
  ]
  rendererClasses [&NiagaraSpriteRendererProperties]
  renderers [
   {rendererClass=&NiagaraSpriteRendererProperties}
   {rendererClass=&NiagaraSpriteRendererProperties, rendererIndex=1}
  ]
 }
 Sparkles {
  emitterUpdateScript [
   &EmitterState(
    Enable Distance Culling=true
    Enable Visibility Culling=true
    Life Cycle Mode=::NewEnumerator1
    Loop Behavior=::NewEnumerator1
    Loop Duration=2
    MaxDistance=2500
    Module.bApplySpawnCountScaleByDistance=true
    Module.bUseMaxDistance=true
    Module.bUseMinDistance=false
    Scalability Mode=::NewEnumerator1
    Scale Spawn Count=true
    UseLoopDelay=false
   )
   &SpawnRate(SpawnRate=75, Use Spawn Probability=false)
   &SpawnRate(SpawnGroup=1, SpawnRate=10, Use Spawn Probability=false) {module=SpawnRate001}
  ]
  particleSpawnScript [
   &InitializeParticle(
    Color=&MergeHelper_ColorSwitch(
     0=&RandomRangeLinearColor(Minimum=<rgba 1 0.4683 0.3281 0.9957>, Maximum=<rgba 2 1.1 0.2 1>)
     1=&RandomRangeLinearColor(Minimum=<rgba 0.5677 0.7632 1 0.9957>, Maximum=<rgba 0.3438 0.8334 2 1>)
     2=&RandomRangeLinearColor(Minimum=<rgba 1 0.814 0.7947 1>, Maximum=<rgba 2 1.6231 1.5861 1>)
     3=&RandomRangeLinearColor(Minimum=<rgba 0.9323 0.977 1 0.9957>, Maximum=<rgba 0.2 1.0139 2 1>)
     4=&RandomRangeLinearColor(Minimum=<rgba 0.5885 1 0.9678 1>, Maximum=<rgba 0.0626 1.8484 2 1>)
     5=&RandomRangeLinearColor(Minimum=<rgba 0.5885 1 0.9678 1>, Maximum=<rgba 1.9066 0.6672 2 1>)
     Index=&MakeIntFromFloat(Float=$User.VariantPicker, Conversion Mode=::NewEnumerator1)
     Num=::NewEnumerator1
    )
    Lifetime Max=0.33
    Lifetime Min=0.15
    Lifetime Mode=::NewEnumerator1
    Sprite Rotation Mode=::NewEnumerator2
    Sprite Size Mode=::NewEnumerator3
    Uniform Sprite Size=&MergeHelper_FloatSwitch(
     0=&RandomRangeFloat(Minimum=1.5, Maximum=8)
     1=&RandomRangeFloat(Minimum=15, Maximum=20)
     2=?
     3=?
     Index=&SpawnGroup()
     Num=::NewEnumerator0
    )
    UsePositionOffset=true
    Write Material Random=true
   )
   &ShapeLocation(MaskSpawnGroup=false, Override Seed=false, Sphere Radius=6, Transform Order:ENiagara_TransformOrder=::NewEnumerator1, UseOwnerScale=true) {module=ShapeLocation001}
   &AddVelocity(
    Module.Use Falloff Distance=false
    Override Seed=false
    Velocity Mode=::NewEnumerator1
    Velocity Speed=&RandomRangeFloat(Minimum=0, Maximum=50)
   ) {module=AddVelocity001}
   &SocketLocation(SamplePosOffset=false, Socket Reader=$System.Socket Reader, Socket Sampling Mode=::NewEnumerator2, VelocitySpeedLimit=false)
  ]
  particleUpdateScript [
   &ParticleState()
   &SpriteRotationRate(Rotation Rate=&RandomRangeFloat(Minimum=-360, Maximum=360))
   &Drag(Drag=10, Use Linear Drag=true, Use Rotational Drag=true)
   &CurlNoiseForce(Noise Strength=200, Pan Noise=false)
   &SolveForcesAndVelocity(Clamp Velocity=false, Limit Acceleration=false)
   &ScaleSpriteSize(Uniform Curve Sprite Scale=DataInterface() {propertyValues {Curve {
       defaultValue=0
       keys [{interpMode=RCIM_Cubic, tangentMode=RCTM_User, time=0, value=1}, {interpMode=RCIM_Cubic, tangentMode=RCTM_User, time=1, value=0}]
      }}})
   &ScaleColor(
    Scale Alpha=&FNTimeOfDayScalar(Night Scale=1, Morning Scale=1, Evening Scale=1)
    Scale RGB=&MakeVectorFromLinearColorRGB(Color=$System.TOD)
    ScaleA=true
    ScaleRGB=true
   )
  ]
  rendererClasses [&NiagaraSpriteRendererProperties]
  renderers [{rendererClass=&NiagaraSpriteRendererProperties}]
 }
 Trail_Bright {
  emitterUpdateScript [
   &EmitterState(Life Cycle Mode=::NewEnumerator1, Loop Behavior=::NewEnumerator1, Loop Duration=0.5, UseLoopDelay=false)
   &FN_SpawnAnimTrail()
  ]
  particleSpawnScript [
   &FN_InitializeAnimTrailBoneIndecies(First Bone or Socket Index=1, Second Bone or Socket Index=0)
   &FN_InitializeAnimTrail(
    AnimTrailWidth=1.6
    Color=&MergeHelper_ColorSwitch(
     0=<rgba 0.9823 0.4964 0.1484 0.2>
     1=<rgba 0.3471 0.5048 0.9948 0.3>
     2=<rgba 0.2831 0.2195 0.227 1>
     3=?
     Index=&MakeIntFromFloat(Float=$User.VariantPicker, Conversion Mode=::NewEnumerator1)
     Num=::NewEnumerator1
    )
    Lifetime=0.2
   )
  ]
  particleUpdateScript [
   &FN_UpdateAnimTrail()
   &ParticleState()
   &ScaleRibbonWidth(Ribbon Width Scale=&FloatFromCurve(FloatCurve=DataInterface() {propertyValues {Curve {
        defaultValue=1
        keys [
         {arriveTangent=-2.1282, interpMode=RCIM_Cubic, leaveTangent=-2.1282, tangentMode=RCTM_User, time=0, value=1}
         {interpMode=RCIM_Cubic, tangentMode=RCTM_User, time=1, value=0}
        ]
       }}})) {enabled=false}
   &ScaleColor(
    Scale Alpha=&Multiply_Float(
     A=&FloatFromCurve(
      FloatCurve=DataInterface() {propertyValues {Curve {keys [
          {arriveTangent=0.2188, interpMode=RCIM_Cubic, leaveTangent=-2.2593, tangentMode=RCTM_Break, time=0, value=0}
          {arriveTangent=-5.7332, interpMode=RCIM_Cubic, leaveTangent=-5.7332, tangentMode=RCTM_User, time=0.05, value=0.9}
          {arriveTangent=-0.0239, interpMode=RCIM_Cubic, tangentMode=RCTM_Break, time=0.4996, value=0}
         ]}}}
      Scale Curve=&FN_FrontEndFloatModifier(Gameplay=1, Front End=$User.FrontEndAlpha)
     )
     B=&FNTimeOfDayScalar(Night Scale=3, Morning Scale=1, Day Scale=0.6, Evening Scale=1)
    )
    Scale RGB=&Multiply_Vector(
     A=&MakeVectorFromLinearColorRGB(Color=&ColorFromCurve(ColorCurve=DataInterface() {propertyValues {
         AlphaCurve {keys []}
         BlueCurve {keys [<at 0 3>,<at 0.4832 1>]}
         GreenCurve {keys [<at 0 3>,<at 0.4832 1>]}
         RedCurve {keys [<at 0 3>,<at 0.4832 1>]}
        }}))
     B=&MakeVectorFromLinearColorRGB(Color=$System.TOD)
    )
    ScaleA=true
    ScaleRGB=true
   )
   &DynamicMaterialParameters(
    Index 0 Param 1=&FloatFromCurve(FloatCurve=DataInterface() {propertyValues {Curve {keys [{interpMode=RCIM_Cubic, tangentMode=RCTM_User, time=0.2, value=1}, {interpMode=RCIM_Cubic, tangentMode=RCTM_User, time=1, value=0}]}}})
    Index 0 Param 2=1
    Index 0 Param 3=&Multiply_Float(
     A=&SelectFloatFromArray(
      Direct Array Index=&SpawnGroup()
      Float Selection Array=DataInterface() {propertyValues {FloatData [-1.95,-2,-2.1,-2.05]}}
     )
     B=0.25
    )
    Index 0 Param 4=1
    Index 1 Param 1=1
    Index 1 Param 2=0.05
    Module.Param0WriteEnabled=true
    Module.Param0WriteEnabled1=true
    Module.Param1WriteEnabled=true
    Module.Param1WriteEnabled1=true
    Module.Param2WriteEnabled=true
    Module.Param2WriteEnabled1=true
    Module.Param3WriteEnabled=true
    Module.Param3WriteEnabled1=false
    Write Parameter Index 1=true
   )
  ]
  rendererClasses [&NiagaraRibbonRendererProperties]
  renderers [{rendererClass=&NiagaraRibbonRendererProperties}]
 }
 Trail_Edge {
  emitterUpdateScript [
   &EmitterState(Life Cycle Mode=::NewEnumerator1, Loop Behavior=::NewEnumerator1, Loop Duration=0.5, UseLoopDelay=false)
   &FN_SpawnAnimTrail()
  ]
  particleSpawnScript [
   &FN_InitializeAnimTrailBoneIndecies()
   &FN_InitializeAnimTrail(Color=<rgba 1.5 1.5 1.5 1>, Lifetime=0.3)
  ]
  particleUpdateScript [
   &FN_UpdateAnimTrail()
   &ParticleState()
   &ScaleRibbonWidth(Ribbon Width Scale=&FloatFromCurve(FloatCurve=DataInterface() {propertyValues {Curve {keys [
         {arriveTangent=-0.0514, interpMode=RCIM_Cubic, leaveTangent=-0.0514, tangentMode=RCTM_User, time=0, value=1}
         {arriveTangent=-1.0387, interpMode=RCIM_Cubic, leaveTangent=-1.0387, tangentMode=RCTM_User, time=1, value=0.3}
        ]}}}))
   &ScaleColor(
    Scale Alpha=&Multiply_Float(
     A=&FloatFromCurve(
      FloatCurve=DataInterface() {propertyValues {Curve {keys [
          {interpMode=RCIM_Cubic, leaveTangent=-2.3646, tangentMode=RCTM_Break, time=0, value=1}
          {arriveTangent=-0.0239, interpMode=RCIM_Cubic, tangentMode=RCTM_Break, time=1, value=0}
         ]}}}
      Scale Curve=&FN_FrontEndFloatModifier(Gameplay=1, Front End=$User.FrontEndAlpha)
     )
     B=&FNTimeOfDayScalar(Night Scale=1, Morning Scale=0.6, Day Scale=0.6, Evening Scale=0.6)
    )
    Scale RGB=&Multiply_Vector(
     A=&MergeHelper_VectorSwitch(
      0=&MakeVectorFromLinearColorRGB(Color=&ColorFromCurve(ColorCurve=DataInterface() {propertyValues {
          AlphaCurve {keys []}
          BlueCurve {keys [<at 0 0.1354>,<at 0.6365 0.0738>]}
          GreenCurve {keys [<at 0 0.9137>,<at 0.6365 0.2724>]}
          RedCurve {keys [<at 0 2>,<at 0.6365 0.7084>]}
         }}))
      1=&MakeVectorFromLinearColorRGB(Color=&ColorFromCurve(ColorCurve=DataInterface() {propertyValues {
          AlphaCurve {keys []}
          BlueCurve {keys [<at 0 2>,<at 0.6344 0.4508>]}
          GreenCurve {keys [<at 0 1.1169>,<at 0.6344 0.0482>]}
          RedCurve {keys [<at 0 0.6458>,<at 0.6344 0.0972>]}
         }}))
      2=&MakeVectorFromLinearColorRGB(Color=&ColorFromCurve(ColorCurve=DataInterface() {propertyValues {
          AlphaCurve {keys []}
          BlueCurve {keys [<at 0 0.5339>,<at 0.6344 0.0833>]}
          GreenCurve {keys [<at 0 0.5469>,<at 0.6344 0.0595>]}
          RedCurve {keys [<at 0 0.6719>,<at 0.6344 0.0758>]}
         }}))
      3=?
      Index=&MakeIntFromFloat(Float=$User.VariantPicker, Conversion Mode=::NewEnumerator1)
      Num=::NewEnumerator1
     )
     B=&MakeVectorFromLinearColorRGB(Color=$System.TOD)
    )
    ScaleA=true
    ScaleRGB=true
   )
  ]
  rendererClasses [&NiagaraRibbonRendererProperties]
  renderers [{rendererClass=&NiagaraRibbonRendererProperties}]
 }
 Trail_Soft {
  emitterUpdateScript [
   &EmitterState(Life Cycle Mode=::NewEnumerator1, Loop Behavior=::NewEnumerator1, Loop Duration=0.5, UseLoopDelay=false)
   &FN_SpawnAnimTrail()
  ]
  particleSpawnScript [
   &FN_InitializeAnimTrailBoneIndecies(First Bone or Socket Index=1, Second Bone or Socket Index=0)
   &FN_InitializeAnimTrail(
    AnimTrailWidth=2
    Color=&SelectLinearColorFromArray(
     Array Sampling Mode=::NewEnumerator1
     Direct Array Index=&MakeIntFromFloat(Float=$User.VariantPicker, Conversion Mode=::NewEnumerator1)
     Color Selection Array=DataInterface() {propertyValues {ColorData [<rgba 1 1 1 1>, <rgba 1 1 1 1>, <rgba 0.0395 0.0354 0.0625 2.2>]}}
     Random Seed=?
    )
    Lifetime=0.3
   )
  ]
  particleUpdateScript [
   &FN_UpdateAnimTrail()
   &ParticleState()
   &ScaleRibbonWidth(Ribbon Width Scale=&FloatFromCurve(FloatCurve=DataInterface() {propertyValues {Curve {keys [
         {arriveTangent=-0.8056, interpMode=RCIM_Cubic, leaveTangent=-0.8056, tangentMode=RCTM_User, time=0, value=1}
         {arriveTangent=-0.7153, interpMode=RCIM_Cubic, leaveTangent=-0.7153, tangentMode=RCTM_User, time=1, value=0.3}
        ]}}}))
   &ScaleColor(
    Scale Alpha=&Multiply_Float(
     A=&FloatFromCurve(FloatCurve=DataInterface() {propertyValues {Curve {keys [
          {arriveTangent=-0.0116, interpMode=RCIM_Cubic, leaveTangent=-1.6168, tangentMode=RCTM_Break, time=0, value=1}
          {arriveTangent=0.0949, interpMode=RCIM_Cubic, tangentMode=RCTM_Break, time=1, value=0.2}
         ]}}})
     B=&MergeHelper_FloatSwitch(
      0=&FNTimeOfDayScalar(Night Scale=0.6, Morning Scale=1, Evening Scale=0.8)
      1=&FNTimeOfDayScalar(Night Scale=1, Morning Scale=0.6, Day Scale=0.6, Evening Scale=0.6)
      2=0.6
      3=?
      Index=&MakeIntFromFloat(Float=$User.VariantPicker, Conversion Mode=::NewEnumerator1)
      Num=::NewEnumerator1
     )
    )
    Scale RGB=&Multiply_Vector(
     A=&MakeVectorFromLinearColorRGB(Color=&ColorFromCurve(ColorCurve=DataInterface() {propertyValues {
         AlphaCurve {keys []}
         BlueCurve {keys [<at 0.2237 1>,<at 0.4832 1>]}
         GreenCurve {keys [<at 0.2237 1>,<at 0.4832 1>]}
         RedCurve {keys [<at 0.2237 1>,<at 0.4832 1>]}
        }}))
     B=&MakeVectorFromLinearColorRGB(Color=$System.TOD)
    )
    ScaleA=true
    ScaleRGB=true
   )
   &DynamicMaterialParameters(
    Index 0 Param 1=&FloatFromCurve(FloatCurve=DataInterface() {propertyValues {Curve {keys [
         {arriveTangent=0.2015, interpMode=RCIM_Cubic, leaveTangent=0.2015, tangentMode=RCTM_User, time=0, value=1}
         {arriveTangent=-1.5941, interpMode=RCIM_Cubic, leaveTangent=-1.5941, tangentMode=RCTM_User, time=0.3, value=0.5}
         {interpMode=RCIM_Cubic, tangentMode=RCTM_User, time=1, value=0}
        ]}}})
    Index 0 Param 3=-0.8
    Index 0 Param 4=&FloatFromCurve(FloatCurve=DataInterface() {propertyValues {Curve {keys [
         {interpMode=RCIM_Cubic, tangentMode=RCTM_User, time=0, value=1}
         {arriveTangent=-0.6219, interpMode=RCIM_Cubic, leaveTangent=-0.6219, tangentMode=RCTM_User, time=1, value=0}
        ]}}})
    Module.Param0WriteEnabled=true
    Module.Param1WriteEnabled=true
    Module.Param2WriteEnabled=true
    Module.Param3WriteEnabled=true
   )
  ]
  rendererClasses [&NiagaraRibbonRendererProperties]
  renderers [
   {rendererClass=&NiagaraRibbonRendererProperties}
   {rendererClass=&NiagaraRibbonRendererProperties, rendererIndex=1}
   {rendererClass=&NiagaraRibbonRendererProperties, rendererIndex=2}
  ]
 }
}
summary {
 systemName=NS_Pickaxe_RowLiaison_Trail
 emitters [
  {emitterName=Trail_Soft, rendererClasses [&NiagaraRibbonRendererProperties]}
  {emitterName=Trail_Bright, rendererClasses [&NiagaraRibbonRendererProperties]}
  {emitterName=Trail_Edge, rendererClasses [&NiagaraRibbonRendererProperties]}
  {emitterName=Glitter, rendererClasses [&NiagaraSpriteRendererProperties]}
  {emitterName=Sparkles, rendererClasses [&NiagaraSpriteRendererProperties]}
 ]
 userVariables [
  User.bIsFrontEnd {type=NiagaraBool}
  User.bIsFrontEndPreview {type=NiagaraBool}
  User.FrontEndAlpha {type=NiagaraFloat, defaultValue=0.8, description=""}
  User.RibbonIDOffset {type=NiagaraInt32}
  User.VariantPicker {type=NiagaraFloat, defaultValue=0, description=""}
 ]
}
